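# Unlit, two-sided translucent material for Niagara mesh particles, named as the edge glow effect of the Water Molecules pickaxe.
# Particle color feeds an emissive adjustment function, with no textures sampled.
# A procedural mask multiplies particle alpha by a powered V gradient, a view-angle fade and a U falloff.
# The U falloff is a saturated, scaled tent gradient that softens both side edges of the mesh UVs.
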Material /BRCosmetics/Weapons/FORT_Melee/Pickaxe_WaterMolecules/FX/M_FX_WaterMolecules_EdgeGlow
graph {
 properties {
  bUsedWithNiagaraMeshParticles=true
  blendMode=BLEND_Translucent
  pixelDepthOffsetMode=PDOM_Legacy
  shadingModel=MSM_Unlit
  twoSided=true
  usedShadingModels=Unlit
 }
 nodes [
  n1 = ParticleColor()
  MaterialFunctionCall(Emissive=@n1.RGB) {materialFunction=&MF_FX_EmissiveAdjustments}
  n2 = Power(
   Base=MaterialFunctionCall(Float2=TextureCoordinate()) {materialFunction=&BreakOutFloat2Components}.G
   Exp=ScalarParameter() {defaultValue=5}
  )
  n3 = MaterialFunctionCall("FadeSensitivity (optional)"=ScalarParameter() {defaultValue=1.5, parameterName=FadeBasedOnViewAngle}) {materialFunction=&FadeBasedOnViewAngle}
  n4 = TextureCoordinate()
  n5 = Abs(None=ConstantBiasScale(None=MaterialFunctionCall(Float2=@n4) {materialFunction=&BreakOutFloat2Components}.R) {bias=-0.5, scale=2})
  n6 = Saturate(None=Multiply(A=OneMinus(None=@n5), B=ScalarParameter() {defaultValue=10, parameterName=Param_1}))
  Multiply(A=@n1.A, B=Multiply(A=Multiply(A=@n2, B=@n3.Result), B=Multiply(A=@n6)))
 ]
}
